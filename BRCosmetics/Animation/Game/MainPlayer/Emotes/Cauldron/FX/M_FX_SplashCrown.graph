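# Masked, lit Niagara sprite and mesh particle material, named for the splash crown effect of the Cauldron emote.
# Scaled UVs, offset by particle random and panned in V by a dynamic parameter, sample borrowed Asteria mud mask and normal textures.
# An optional dissolve adds particle alpha to invertible cell noise, while particle color feeds an emissive adjustment function.
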Material /BRCosmetics/Animation/Game/MainPlayer/Emotes/Cauldron/FX/M_FX_SplashCrown
graph {
 properties {
  allowTranslucentCustomDepthWrites=true
  bUsedWithNiagaraMeshParticles=true
  bUsedWithNiagaraSprites=true
  blendMode=BLEND_Masked
  pixelDepthOffsetMode=PDOM_Legacy
  usedShadingModels=Default Lit
 }
 nodes [
  n1 = ParticleColor()
  MaterialFunctionCall(Emissive=@n1.RGB) {materialFunction=&MF_FX_EmissiveAdjustments}
  n2 = Multiply(
   A=TextureCoordinate()
   B=ComponentMask(None=VectorParameter() {parameterName=UV Scale, defaultValue=<rgba 0.3 0.3 0 0>}.RGB) {g=true, r=true}
  )
  n3 = DynamicParameter() {paramNames [PrevPosX, PrevPosY, PrevPosZ, UV V Pan]}
  Multiply(A=@n2, B=@n3.UV V Pan)
  n4 = TextureSampleParameter2D(UVs=Multiply(A=TextureCoordinate(), B=ScalarParameter() {defaultValue=1, parameterName=Dissolve Scale})) {parameterName=Dissolve Texture, samplerType=SAMPLERTYPE_LinearGrayscale, texture=/Game/Effects/Fort_Effects/Textures/Tile/T_CellNoise_Tile_M}
  n5 = StaticSwitchParameter(True=OneMinus(None=@n4.R), False=@n4.R) {parameterName=Invert Dissolve Tex}
  StaticSwitchParameter(True=Saturate(None=Add(A=@n5, B=@n1.A)), False=Constant() {r=1}) {parameterName=Use Dissolve}
  n6 = ScalarParameter() {defaultValue=0.75, parameterName=UV Mul}
  n7 = Add(A=Add(A=ParticleRandom(), B=Multiply(A=@n6, B=@n2)), B=Multiply(A=Add(A=@n2, B=@n3.UV V Pan), B=Constant2Vector() {g=1}))
  TextureSample(UVs=@n7) {samplerType=SAMPLERTYPE_Masks, texture=/Game/Environments/Asteria/Terrain/Textures/T_Asteria_SmoothMud_S}
  TextureSample(UVs=@n7) {samplerType=SAMPLERTYPE_Normal, texture=/Game/Environments/Asteria/Terrain/Textures/T_Asteria_MudGround_A_N}
 ]
}
