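# Niagara system for the Cauldron emote's idle effect: a glowing fire under a bubbling, steaming green brew.
# The base combines mesh flames, glow and flickering orange light; the top adds green light, sparkles, steam and falling smoke.
# Each bubble particle computes curve-timed splash, wave and ripple attributes for five renderers.
# Overlay and twin glows follow skeletal-mesh sampling; System.FadeIn, time-of-day and front-end modifiers scale selected brightness and colors.
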NiagaraSystem /BRCosmetics/Animation/Game/MainPlayer/Emotes/Cauldron/FX/NS_Emote_Cauldron_Idle
dependencies {
 usedDataInterfaces [
  &NiagaraDataInterfaceCurve
  &NiagaraDataInterfaceArrayInt32
  &NiagaraDataInterfaceArrayFloat3
  &NiagaraDataInterfaceColorCurve
  &NiagaraDataInterfaceSkeletalMesh
 ]
 usedDynamicInputs [&ClampFloat]
 usedModules [
  &self:SystemSpawnScript.SystemScriptSource.SystemScriptGraph.NiagaraNodeAssignment_0.SetVariables_#1
  &self:SystemSpawnScript.SystemScriptSource.SystemScriptGraph.NiagaraNodeAssignment_1.SetVariables_#2
  &SystemState
 ]
}
emitters {
 BottomGlow {
  emitterUpdateScript [
   &EmitterState()
   &SpawnRate(SpawnRate=4, Use Spawn Probability=false)
  ]
  particleSpawnScript [&InitializeParticle(
    Lifetime Max=0.7
    Lifetime Min=0.45
    Lifetime Mode=::NewEnumerator1
    Position Offset=<0 0 10>
    Sprite Size Mode=::NewEnumerator3
    Uniform Sprite Size=50
    UsePositionOffset=true
    Write Material Random=true
   ) {module=InitializeParticle001}]
  particleUpdateScript [
   &ParticleState()
   &CameraOffset(Camera Offset Amount=15)
   &ScaleColor(
    Scale Alpha=&Multiply_Float(
     A=&Multiply_Float(
      A=&Multiply_Float(A=0.2, B=&FN_FrontEndFloatModifier(Gameplay=1, Front End=0.6))
      B=&FNTimeOfDayScalar(Night Scale=0.6, Morning Scale=1, Evening Scale=1)
     )
     B=$System.FadeIn
    )
    Scale RGB=&MakeVectorFromLinearColorRGB(Color=<rgba 3 0.8837 0.1142 1>)
    ScaleA=true
    ScaleRGB=true
   )
   &DynamicMaterialParameters(
    Index 0 Param 1=4
    Index 0 Param 2=1
    Index 0 Param 3=1
    Index 0 Param 4=1
    Module.Param0WriteEnabled=true
    Module.Param1WriteEnabled=true
    Module.Param2WriteEnabled=true
    Module.Param3WriteEnabled=true
   )
   &JitterPosition(Jitter Amount=0.5)
  ]
  rendererClasses [&NiagaraSpriteRendererProperties]
  renderers [{rendererClass=&NiagaraSpriteRendererProperties}]
 }
 BottomLight {
  emitterUpdateScript [
   &EmitterState(
    Inactive Response=::NewEnumerator1
    Life Cycle Mode=::NewEnumerator1
    Loop Behavior=::NewEnumerator1
    Loop Duration Mode=::NewEnumerator1
    UseLoopDelay=false
   )
   &SpawnBurst_Instantaneous(Use Spawn Probability=false, UseLoopCountLimit=false)
  ]
  particleSpawnScript [
   &InitializeParticle(Color Mode=::NewEnumerator0, Lifetime=5, Sprite Size Mode=::NewEnumerator3, UsePositionOffset=false, Write Material Random=true)
   &SystemLocation(MaskSpawnGroup=false, Offset=<0 0 10>)
  ]
  particleUpdateScript [
   &ParticleState(Kill Particles When Lifetime Has Elapsed=false, Loop Particles Lifetime=true)
   &FN_Light_Attributes(
    Color=&FNTimeOfDayColor(Night Color=<rgba 1 0.439 0 1>, Morning Color=<rgba 1 0.439 0 1>, Day Color=<rgba 1 0.4301 0 1>, Evening Color=<rgba 1 0.439 0 1>)
    Create Color Attribute=true
    Create Exponent Attribute=true
    Create Radius Attribute=true
    Enable Particle Light Enable=true
    Exponent:NiagaraFloat=&Add_Float(
     A=2
     B=&Waveform("Waveform [1]"=::NewEnumerator0, "[1] Frequency"=0.7, "[1] Amplitude Min/Max"=<0 1>, "[1] Scale"=1.5)
    )
    Light Enabled=true
    Radius=&Add_Float(
     A=&Multiply_Float(A=70, B=$System.FadeIn)
     B=&Sine(Normalized Angle=$Emitter.Age, Period=0.7, Scale=4, Bias=0)
    )
    Scale Color=&Multiply_Float(
     A=&Multiply_Float(
      A=20
      B=&Waveform(
       Waveform Count=::NewEnumerator1
       "Waveform [1]"=::NewEnumerator0
       "[1] Frequency"=2
       "[1] Amplitude Min/Max"=<0.5 1>
       "Waveform [2] Blend"=::NewEnumerator2
       "Waveform [2]"=::NewEnumerator1
       "[2] Input Position"=0
       "[2] Phase"=$Emitter.Age
       "[2] Phase Scale"=1
       "[2] Compound Frequency"=2
       "[2] Frequency"=1.3
       "[2] Amplitude Min/Max"=<0.8 1.3>
       "[2] Scale"=1
       "[2] Bias"=0
       "[2] Exponent"=1
       "[2] Pulse Duration"=0.1
       "[2] Clip Triangle Min/Max"=<0 0>
       "[2] Random Seed"=0
      )
     )
     B=$System.FadeIn
    )
   )
  ]
  rendererClasses [&NiagaraLightRendererProperties]
  renderers [{rendererClass=&NiagaraLightRendererProperties, rendererIndex=0}]
 }
 Bubbles {
  emitterUpdateScript [
   &EmitterState(Inactive Response=::NewEnumerator1, Life Cycle Mode=::NewEnumerator1, Loop Duration=1, UseLoopDelay=false)
   &SpawnRate(SpawnRate=2, Use Spawn Probability=false)
   /SetVariables(Emitter.BubbleEmissive:NiagaraFloat=&FNTimeOfDayScalar(Night Scale=0.5, Morning Scale=0.4, Day Scale=0.3, Evening Scale=0.4)) {isSetParameters=true}
  ]
  particleSpawnScript [
   &InitializeParticle(
    Lifetime Min=1.6
    Lifetime Mode=::NewEnumerator1
    Mesh Scale Mode=::NewEnumerator1
    Mesh Uniform Scale Max=0.08
    Mesh Uniform Scale Min=0.045
    Sprite Rotation Mode=::NewEnumerator2
    UsePositionOffset=false
    Write Material Random=true
   )
   &ShapeLocation(
    Disc Coverage=1
    MaskSpawnGroup=false
    Offset=<0 0 94>
    Offset Mode=::NewEnumerator0
    Override Seed=false
    Ring Radius=30
    Shape Primitive=::NewEnumerator4
    UseOwnerScale=true
   )
   /SetVariables(
    Particles.SplashIndex:NiagaraFloat=&RandomRangeFloat(Minimum=0, Maximum=4)
    Particles.SplashVelocity:Vector3f=<0 0 1>
   ) {isSetParameters=true}
  ]
  particleUpdateScript [
   &ParticleState()
   &ScaleMeshSize(Scale Factor=&VectorFromFloat(Value=&FloatFromCurve(FloatCurve=DataInterface() {propertyValues {Curve {keys [<at 0 0.2>,<at 0.7 1>,<at 1 0>]}}})))
   &DynamicMaterialParameters(
    Index 0 Param 1=&FloatFromCurve(FloatCurve=DataInterface() {propertyValues {Curve {keys [
         <at 0 0>
         {interpMode=RCIM_Cubic, leaveTangent=1.2072, tangentMode=RCTM_Break, time=0.45, value=0}
         {arriveTangent=3.283, tangentMode=RCTM_User, time=0.9, value=1}
        ]}}})
    Index 0 Param 4=1.1
    Module.Index0Param0Write=true
    Module.Index0Param1Write=true
    Module.Index0Param2Write=true
    Module.Index0Param3Write=true
   )
   &SolveForcesAndVelocity(Clamp Velocity=false, Limit Acceleration=false)
   &RegionMask(Inner Float=0, Offset:Vector3f=<0 -30 94>, Outer Float=1, Region Falloff=0, Region Radius=25, UseOffset=true, UseRotation=false)
   /SetVariables(
    Particles.SplashDynamicParamenter:Vector4f=&MakeVector4(
     X=&FloatFromCurve(FloatCurve=DataInterface() {propertyValues {Curve {keys [
          <at 0 1>
          {interpMode=RCIM_Cubic, leaveTangent=127.6574, tangentMode=RCTM_Break, time=0.8, value=1}
          {arriveTangent=27.4703, tangentMode=RCTM_User, time=1, value=17}
         ]}}})
     Y=0.7
     Z=0
     W=0
    )
    Particles.SplashPosition:NiagaraPosition=&Add_Vector(
     A=Linked() {linkedVariable {name=Particles.Position, type=NiagaraPosition}}
     B=&MakeVector(
      X=0
      Y=0
      Z=&FloatFromCurve(
       FloatCurve=DataInterface() {propertyValues {Curve {keys [
           {interpMode=RCIM_Cubic, leaveTangent=156.6917, tangentMode=RCTM_Break, time=0.55, value=-10}
           {interpMode=RCIM_Cubic, tangentMode=RCTM_Break, time=0.75, value=5}
           {arriveTangent=-261.0323, interpMode=RCIM_Cubic, tangentMode=RCTM_Break, time=1, value=-50}
          ]}}}
       Scale Curve=&RandomRangeFloat(Minimum=0.5, Maximum=1)
      )
     )
    )
    Particles.SplashSize:Vector2f=&Multiply_Vector2D(
     A=&Multiply_Vector2D(
      A=<0.8 0.8>
      B=&Vector2DFromFloat(Value=&Lerp_Float(A=0, B=1, Alpha=$StackContext.RegionMask.Float))
     )
     B=&Vector2DFromFloat(Value=&FloatFromCurve(
       FloatCurve=DataInterface() {propertyValues {Curve {keys [<at 0.6 0>,<at 0.8 1>]}}}
       Scale Curve=&RandomRangeFloat(Minimum=20, Maximum=25)
      ))
    )
    Particles.SplashVector:Vector3f=<0 0 1>
   ) {isSetParameters=true}
   /SetVariables(
    Particles.WaveColor:LinearColor=&ColorFromCurve(ColorCurve=DataInterface() {propertyValues {
       AlphaCurve {keys [<at 0.6765 0>,<at 0.7462 1>,<at 0.9338 0 RCIM_Cubic>]}
       BlueCurve {keys [<at 0.8093 0.15 RCIM_Cubic>,<at 1 0.0531 RCIM_Cubic>]}
       GreenCurve {keys [<at 0.8093 1 RCIM_Cubic>,<at 1 0.3542 RCIM_Cubic>]}
       RedCurve {keys [<at 0.8093 0.2917 RCIM_Cubic>,<at 1 0.1033 RCIM_Cubic>]}
      }})
    Particles.WaveSize:Vector2f=&Vector2DFromFloat(Value=&MakeFloatFromVector2d(Vector 2D=&Vector2DFromFloat(Value=&FloatFromCurve(
        FloatCurve=DataInterface() {propertyValues {Curve {keys [<at 0 0>,<at 0.7 0>,<at 0.8 1>]}}}
        Scale Curve=&RandomRangeFloat(Minimum=25, Maximum=30)
       ))))
   ) {isSetParameters=true}
   /SetVariables(
    Particles.RippleColor:LinearColor=&MultiplyLinearColor(
     Linear Color=&ColorFromCurve(ColorCurve=DataInterface() {propertyValues {
        AlphaCurve {keys [<at 0.6765 0>,<at 0.7462 0.2>,<at 0.9056 0 RCIM_Cubic>]}
        BlueCurve {keys [<at 0.8093 0.15 RCIM_Cubic>,<at 1 0.0531 RCIM_Cubic>]}
        GreenCurve {keys [<at 0.8093 1 RCIM_Cubic>,<at 1 0.3542 RCIM_Cubic>]}
        RedCurve {keys [<at 0.8093 0.2917 RCIM_Cubic>,<at 1 0.1033 RCIM_Cubic>]}
       }})
     Scale Factor=&MakeVector4_FromLinearColor(LinearColor=&FNTimeOfDayColor(Night Color=<rgba 0.556 0.5008 1 1>, Morning Color=<rgba 1 0.6695 0.6095 1>, Day Color=<rgba 1 1 1 1>, Evening Color=<rgba 1 1 1 1>))
    )
    Particles.RippleFacing:Vector3f=<0 0 1>
    Particles.RippleSize:Vector2f=&Vector2DFromFloat(Value=&FloatFromCurve(FloatCurve=DataInterface() {propertyValues {Curve {keys [<at 0.5 0>,<at 1 1>]}}}, Scale Curve=35))
   ) {isSetParameters=true}
   /SetVariables(
    Particles.Ripple2Color:LinearColor=&MultiplyLinearColor(
     Linear Color=&ColorFromCurve(ColorCurve=DataInterface() {propertyValues {
        AlphaCurve {keys [<at 0.8143 0>,<at 0.8876 0.1>,<at 0.9661 0 RCIM_Cubic>]}
        BlueCurve {keys [<at 0.8093 0.5328 RCIM_Cubic>,<at 1 0.109 RCIM_Cubic>]}
        GreenCurve {keys [<at 0.8093 1 RCIM_Cubic>,<at 1 0.3542 RCIM_Cubic>]}
        RedCurve {keys [<at 0.8093 0.4711 RCIM_Cubic>,<at 1 0.1498 RCIM_Cubic>]}
       }})
     Scale Factor=&MakeVector4_FromLinearColor(LinearColor=&FNTimeOfDayColor(Night Color=<rgba 0.556 0.5008 1 1>, Morning Color=<rgba 1 0.6695 0.6095 1>, Day Color=<rgba 1 1 1 1>, Evening Color=<rgba 1 1 1 1>))
    )
    Particles.Ripple2Size:Vector2f=&Vector2DFromFloat(Value=&FloatFromCurve(
      FloatCurve=DataInterface() {propertyValues {Curve {keys [
          {interpMode=RCIM_Cubic, leaveTangent=0.6678, tangentMode=RCTM_Break, time=0.6, value=0}
          {arriveTangent=2.5069, tangentMode=RCTM_User, time=1, value=0.65}
         ]}}}
      Scale Curve=30
     ))
   ) {isSetParameters=true}
   &ScaleMeshSize(Scale Factor=&VectorFromFloat(Value=&Lerp_Float(A=0, B=1, Alpha=$StackContext.RegionMask.Float))) {module=ScaleMeshSize001}
  ]
  rendererClasses [&NiagaraMeshRendererProperties, &NiagaraSpriteRendererProperties]
  renderers [
   {rendererClass=&NiagaraMeshRendererProperties, rendererIndex=0}
   {rendererClass=&NiagaraSpriteRendererProperties, rendererIndex=1}
   {rendererClass=&NiagaraSpriteRendererProperties, rendererIndex=2}
   {rendererClass=&NiagaraSpriteRendererProperties, rendererIndex=3}
   {rendererClass=&NiagaraSpriteRendererProperties, rendererIndex=4}
  ]
 }
 CandyCorn_Fire {
  emitterUpdateScript [
   &EmitterState()
   &SpawnBurst_Instantaneous(Spawn Count=5, Use Spawn Probability=false, UseLoopCountLimit=false)
   &SpawnRate(SpawnRate=20, Use Spawn Probability=false)
  ]
  particleSpawnScript [
   &InitializeParticle(
    Color Mode=::NewEnumerator0
    Lifetime=&RandomRangeFloat(Minimum=0.5, Maximum=0.8)
    Mesh Scale=&Multiply_Vector(
     A=&RandomRangeVector(Minimum=<0.8 0.8 1>, Maximum=<1 1 1.5>)
     B=&VectorFromFloat(Value=&RandomRangeFloat(Minimum=1, Maximum=1.9))
    )
    Mesh Scale Mode=::NewEnumerator4
    UsePositionOffset=false
    Write Material Random=true
   ) {module=InitializeParticle001}
   &ShapeLocation(
    Disc Coverage=1
    MaskSpawnGroup=false
    Offset=<0 0 5>
    Offset Mode=::NewEnumerator0
    Override Seed=false
    Ring Radius=&Multiply_Float(
     A=9
     B=&FloatFromCurve(
      FloatCurve=DataInterface() {propertyValues {Curve {keys [<at 0.35 0.8 RCIM_Cubic>, {arriveTangent=1.3333, interpMode=RCIM_Cubic, leaveTangent=1.3333, time=0.4, value=1}, <at 0.5 1 RCIM_Cubic>]}}}
      CurveIndex=$Emitter.Age
     )
    )
    Shape Primitive=::NewEnumerator4
    UseOwnerScale=true
   )
   &InitialMeshOrientation(Mesh Orientation Mode=::NewEnumerator2, Orientation Vector=<0 0 -1>, Use Rotation Vector=false)
   &AddVelocity(Velocity=<0 0 1>, Velocity Speed Scale=&RandomRangeFloat(Minimum=0, Maximum=0))
  ]
  particleUpdateScript [
   &ParticleState()
   &SolveForcesAndVelocity(Clamp Velocity=false, Limit Acceleration=false)
   &ParticleState() {module=ParticleState001}
   /SetVariables(Particles.MainColor:LinearColor=&MultiplyLinearColor(
     Linear Color=&MultiplyLinearColor(Linear Color=<rgba 1 0.5779 0 1>, Scale Factor=<1 1 1 1>)
     Scale Factor=&MakeVector4_FromFloat(Float=&Waveform(
       Waveform Count=::NewEnumerator1
       Global Input Position=&ReturnNormalizedExecIndex(Normalized Index Scale=9)
       "Waveform [1]"=::NewEnumerator12
       "[1] Frequency"=2.65
       "[1] Amplitude Min/Max"=<0.7 1>
       "Waveform [2] Blend"=::NewEnumerator5
       "[2] Interpolation"=0.5
       "[2] Invert Interpolation"=false
       "Waveform [2]"=::NewEnumerator12
       "[2] Input Position"=0
       "[2] Phase"=$Emitter.Age
       "[2] Phase Scale"=1
       "[2] Compound Frequency"=2
       "[2] Frequency"=16
       "[2] Amplitude Min/Max"=<0.75 1>
       "[2] Scale"=1
       "[2] Bias"=0
       "[2] Exponent"=1
       "[2] Pulse Duration"=0.1
       "[2] Clip Triangle Min/Max"=<0 0>
       "[2] Random Seed"=0
      ))
    )) {isSetParameters=true}
   &ScaleMeshSize(Scale Factor=&Multiply_Vector(
     A=&MakeVector(
      X=&FloatFromCurve(FloatCurve=DataInterface() {propertyValues {Curve {keys [
           {interpMode=RCIM_Cubic, tangentMode=RCTM_Break, time=0, value=1}
           {interpMode=RCIM_Cubic, tangentMode=RCTM_Break, time=0.3, value=1.15}
           {interpMode=RCIM_Cubic, tangentMode=RCTM_Break, time=1, value=0.5}
          ]}}})
      Y=&FloatFromCurve(FloatCurve=DataInterface() {propertyValues {Curve {keys [
           {interpMode=RCIM_Cubic, tangentMode=RCTM_Break, time=0, value=1}
           {interpMode=RCIM_Cubic, tangentMode=RCTM_Break, time=0.3, value=1.15}
           {interpMode=RCIM_Cubic, tangentMode=RCTM_Break, time=1, value=0.5}
          ]}}})
      Z=&FloatFromCurve(FloatCurve=DataInterface() {propertyValues {Curve {keys [<at 0 0.3>,<at 1 1>]}}})
     )
     B=&Multiply_VectorByFloat(
      Vector=&VectorFromFloat(Value=&Waveform(
        Global Input Position=&RandomRangeFloat(Minimum=0, Maximum=1)
        "Waveform [1]"=::NewEnumerator0
        "[1] Frequency"=3
        "[1] Amplitude Min/Max"=<1 1.3>
       ))
      Float=&FloatFromCurve(
       FloatCurve=DataInterface() {propertyValues {Curve {keys [<at 0.3 0.5 RCIM_Cubic>, {arriveTangent=2.5, interpMode=RCIM_Cubic, leaveTangent=2.5, time=0.4, value=1.1}, <at 0.5 1 RCIM_Cubic>]}}}
       CurveIndex=$Emitter.Age
      )
     )
    ))
   /SetVariables(
    Particles.CapsuleGlowCameraOffset=-10
    Particles.CapsuleGlowColor:LinearColor=&MultiplyLinearColor(
     Linear Color=$StackContext.MainColor
     Scale Factor=&MakeFloat4FromVectorAndFloat(
      "Vector (RGB)"=&VectorFromFloat(Value=1)
      "Float (Alpha)"=&FloatFromCurve(
       FloatCurve=DataInterface() {propertyValues {Curve {keys [
           <at 0 0 RCIM_Cubic>
           {arriveTangent=1.5152, interpMode=RCIM_Cubic, leaveTangent=1.5152, time=0.3, value=1}
           {arriveTangent=-1.4286, interpMode=RCIM_Cubic, leaveTangent=-1.4286, time=0.66, value=1}
           <at 1 0 RCIM_Cubic>
          ]}}}
       Scale Curve=0.5
      )
     )
    )
    Particles.CapsuleSpriteSize:Vector2f=&Vector2DFromFloat(Value=20)
   ) {isSetParameters=true}
   /SetVariables(Particles.CapsuleSpriteAlignment:Vector3f=$Particles.Velocity) {isSetParameters=true}
   /SetVariables(Particles.CapsuleSpriteDynamic:Vector4f=&MakeFloat4FromVectorAndFloat(
     "Vector (RGB)"=&NormalizeVector(Vector To Normalize=$Particles.CapsuleSpriteAlignment, Safe Normalize=true, Safe Fallback Threshold=0, Safe Fallback Vector=<0 0 0.001>)
     "Float (Alpha)"=20
    )) {isSetParameters=true}
  ]
  rendererClasses [&NiagaraMeshRendererProperties]
  renderers [{rendererClass=&NiagaraMeshRendererProperties, rendererIndex=0}]
 }
 CauldronOverlay {
  emitterUpdateScript [
   &EmitterState(
    Life Cycle Mode=::NewEnumerator1
    Loop Behavior=::NewEnumerator1
    Loop Duration Mode=::NewEnumerator1
    Scalability Mode=::NewEnumerator1
    UseLoopDelay=false
   )
   &SpawnBurst_Instantaneous(Use Spawn Probability=false, UseLoopCountLimit=false)
  ]
  particleSpawnScript [&InitializeParticle(Color=<rgba 1 0.3167 0 1>, Lifetime=5, Mesh Scale Mode=::NewEnumerator3, UsePositionOffset=false, Write Material Random=true)]
  particleUpdateScript [
   &ParticleState(Kill Particles When Lifetime Has Elapsed=false)
   &SkeletalMeshLocation(Bone Sampling Mode=::NewEnumerator4, MaskSpawnGroup=false, Sampled Position Offset=<0 0 -88.1>, Skeletal Mesh=$System.SkeletalMesh)
   &ScaleColor(
    Scale Alpha=&Multiply_Float(
     A=&Waveform(
      Waveform Count=::NewEnumerator1
      "Waveform [1]"=::NewEnumerator0
      "[1] Frequency"=0.7
      "[1] Amplitude Min/Max"=<0.7 1>
      "Waveform [2] Blend"=::NewEnumerator2
      "Waveform [2]"=::NewEnumerator1
      "[2] Input Position"=0
      "[2] Phase"=$Emitter.Age
      "[2] Phase Scale"=1
      "[2] Compound Frequency"=2
      "[2] Frequency"=3
      "[2] Amplitude Min/Max"=<0.8 1>
      "[2] Scale"=1
      "[2] Bias"=0
      "[2] Exponent"=1
      "[2] Pulse Duration"=0.1
      "[2] Clip Triangle Min/Max"=<0 0>
      "[2] Random Seed"=0
     )
     B=$System.FadeIn
    )
    Scale RGB=&VectorFromFloat(Value=&Waveform(
      Waveform Count=::NewEnumerator1
      "Waveform [1]"=::NewEnumerator0
      "[1] Frequency"=1
      "[1] Amplitude Min/Max"=<1 3>
      "[2] Average"=false
      "Waveform [2]"=::NewEnumerator1
      "[2] Input Position"=0
      "[2] Phase"=$Emitter.Age
      "[2] Phase Scale"=1
      "[2] Compound Frequency"=2
      "[2] Frequency"=3
      "[2] Amplitude Min/Max"=<0.5 1>
      "[2] Scale"=1
      "[2] Bias"=0
      "[2] Exponent"=1
      "[2] Pulse Duration"=0.1
      "[2] Clip Triangle Min/Max"=<0 0>
      "[2] Random Seed"=0
     ))
    ScaleA=true
    ScaleRGB=true
   )
  ]
  rendererClasses [&NiagaraMeshRendererProperties]
  renderers [{rendererClass=&NiagaraMeshRendererProperties, rendererIndex=0}]
 }
 FallingSmoke {
  emitterUpdateScript [
   &EmitterState(Life Cycle Mode=::NewEnumerator1, Loop Behavior=::NewEnumerator1, Loop Delay=1, Loop Duration Mode=::NewEnumerator1, UseLoopDelay=false)
   &SpawnRate(SpawnRate=2, Use Spawn Probability=false)
  ]
  particleSpawnScript [
   &InitializeParticle(
    Color Maximum=<rgba 1 1 1 0.15>
    Color Minimum=<rgba 1 1 1 0.1>
    Color Mode=::NewEnumerator2
    Lifetime Max=5
    Lifetime Min=3
    Lifetime Mode=::NewEnumerator1
    Sprite Rotation Mode=::NewEnumerator2
    Sprite Size Mode=::NewEnumerator1
    Uniform Sprite Size Max=50
    Uniform Sprite Size Min=30
    UsePositionOffset=false
    Write Material Random=true
   )
   &ShapeLocation(
    MaskSpawnGroup=false
    Offset=<0 0 94>
    Offset Coordinate Space=::Local
    Offset Mode=::NewEnumerator0
    Override Seed=false
    Ring Radius=&RandomRangeFloat(Minimum=40, Maximum=50)
    Shape Primitive=::NewEnumerator4
    UseOwnerScale=true
   )
  ]
  particleUpdateScript [
   &ParticleState()
   &CameraOffset(Camera Offset Amount=25)
   &GravityForce(Gravity=&MakeVector(
     X=&RandomRangeFloat(Minimum=-1, Maximum=1)
     Y=&RandomRangeFloat(Minimum=-1, Maximum=1)
     Z=&FloatFromCurve(
      FloatCurve=DataInterface() {propertyValues {Curve {keys [
          {arriveTangent=-1.2674, interpMode=RCIM_Cubic, leaveTangent=-1.2674, tangentMode=RCTM_User, time=0, value=1}
          {interpMode=RCIM_Cubic, tangentMode=RCTM_User, time=1, value=0}
         ]}}}
      Scale Curve=&RandomRangeFloat(Minimum=-20, Maximum=-50)
     )
    ))
   &CurlNoiseForce(Noise Frequency=10, Noise Strength=20, Pan Noise=true, Pan Noise Field=<0 0 1>)
   &Drag(Drag=&RandomRangeFloat(Minimum=1, Maximum=3), Use Linear Drag=true, Use Rotational Drag=false)
   &SolveForcesAndVelocity(Clamp Velocity=true, Limit Acceleration=false, Speed Limit=100)
   &ScaleColor(
    Scale Alpha=&FloatFromCurve(
     FloatCurve=DataInterface() {propertyValues {Curve {keys [
         {arriveTangent=4.3482, interpMode=RCIM_Cubic, leaveTangent=4.3482, tangentMode=RCTM_User, time=0, value=0}
         {interpMode=RCIM_Cubic, tangentMode=RCTM_User, time=0.5, value=1}
         {tangentMode=RCTM_User, time=1, value=0}
        ]}}}
     Scale Curve=&Multiply_Float(
      A=&FN_FrontEndFloatModifier(Gameplay=0.666, Front End=0.3)
      B=&FNTimeOfDayScalar(Night Scale=0.3, Morning Scale=1, Day Scale=0.3, Evening Scale=0.6)
     )
    )
    Scale RGB=&MakeVectorFromLinearColorRGB(Color=&MultiplyLinearColor(
      Linear Color=<rgba 0.3911 0.9387 0.2816 1>
      Scale Factor=&MakeVector4_FromLinearColor(LinearColor=&FNTimeOfDayColor(Night Color=<rgba 0.3 0.6229 1 1>, Morning Color=<rgba 1 1 1 1>, Day Color=<rgba 1 1 1 1>, Evening Color=<rgba 1 0.7 0.7444 0>))
     ))
    ScaleA=true
    ScaleRGB=true
   )
   &ScaleSpriteSizeBySpeed(Max Scale Factor=<1.1 1.1>, Velocity Threshold=100)
   &ScaleSpriteSize(Uniform Curve Sprite Scale=DataInterface() {propertyValues {Curve {keys [
        {arriveTangent=0.9769, interpMode=RCIM_Cubic, leaveTangent=0.9769, tangentMode=RCTM_User, time=0, value=0.5}
        {interpMode=RCIM_Cubic, tangentMode=RCTM_User, time=1, value=1}
       ]}}})
   &SubUVAnimation()
  ]
  rendererClasses [&NiagaraSpriteRendererProperties]
  renderers [{rendererClass=&NiagaraSpriteRendererProperties}]
 }
 LamaGlow {
  emitterUpdateScript [
   &EmitterState()
   &SpawnBurst_Instantaneous(Spawn Count=2, Use Spawn Probability=false, UseLoopCountLimit=true)
  ]
  particleSpawnScript [
   &InitializeParticle(
    Lifetime=&RandomRangeFloat(Minimum=0.45, Maximum=0.7)
    Sprite Size Mode=::NewEnumerator3
    Uniform Sprite Size=60
    UsePositionOffset=false
    Write Material Random=true
   ) {module=InitializeParticle001}
   &FN_SetSpawnGroup(Number of Spawn Groups=2)
  ]
  particleUpdateScript [
   &ParticleState(Kill Particles When Lifetime Has Elapsed=false)
   &CameraOffset(Camera Offset Amount=20)
   &ScaleColor(
    Scale Alpha=&Multiply_Float(A=0.333, B=$System.FadeIn)
    Scale RGB=&MakeVectorFromLinearColorRGB(Color=<rgba 0.5711 1 0.1356 1>)
    ScaleA=true
    ScaleRGB=true
   )
   &DynamicMaterialParameters(
    Index 0 Param 1=4
    Index 0 Param 2=1
    Index 0 Param 3=1
    Index 0 Param 4=1
    Module.Param0WriteEnabled=true
    Module.Param1WriteEnabled=true
    Module.Param2WriteEnabled=true
    Module.Param3WriteEnabled=true
   )
   &SkeletalMeshLocation(
    Bone / Socket Index=&SelectIntFromArray(Direct Array Index=$Particles.SpawnGroup, Int Selection Array=DataInterface() {propertyValues {IntData [1,2]}})
    Bone Sampling Mode=::NewEnumerator4
    MaskSpawnGroup=false
    Sampled Position Offset=&SelectVectorFromArray(Direct Array Index=$Particles.SpawnGroup, Vector Selection Array=DataInterface() {propertyValues {FloatData [<-2 3 -13>, <2 3 -13>]}})
    Skeletal Mesh=$System.SkeletalMesh
   )
  ]
  rendererClasses [&NiagaraSpriteRendererProperties]
  renderers [{rendererClass=&NiagaraSpriteRendererProperties}]
 }
 Sparkles {
  emitterUpdateScript [
   &EmitterState()
   &SpawnRate(SpawnRate=1.5, Use Spawn Probability=false)
  ]
  particleSpawnScript [
   &InitializeParticle(
    Color=&MultiplyLinearColor(
     Linear Color=<rgba 10 5.875 0.1 1>
     Scale Factor=&MakeFloat4FromVectorAndFloat(
      "Vector (RGB)"=&VectorFromFloat(Value=&RandomRangeFloat(Minimum=1, Maximum=5))
      "Float (Alpha)"=1
     )
    )
    Lifetime=&UniformRangedFloat(Minimum=2, Maximum=3)
    Sprite Size Mode=::NewEnumerator3
    Uniform Sprite Size=&UniformRangedFloat(Minimum=3, Maximum=8)
    UsePositionOffset=false
    Write Material Random=true
   ) {module=InitializeParticle001}
   &ShapeLocation(
    Disc Coverage=1
    MaskSpawnGroup=false
    Offset=<0 0 95>
    Offset Mode=::NewEnumerator0
    Override Seed=false
    Ring Radius=30
    Shape Primitive=::NewEnumerator4
    UseOwnerScale=true
   )
   &AddVelocity(Velocity=&UniformRangedVector(Minimum=<-30 -30 10>, Maximum=<30 30 100>))
   &CameraOffset(Camera Offset Amount=-10)
   &ApplyInitialForces()
  ]
  particleUpdateScript [
   &ParticleState()
   &ScaleSpriteSize(Scale Factor=&Vector2DFromFloat(Value=&FloatFromCurve(FloatCurve=DataInterface() {propertyValues {Curve {keys [<at -0.0046 0>,<at 0.1 0.9952>,<at 1.0011 0.002>]}}}, Scale Curve=0.8)))
   &Drag(Drag=2, Use Linear Drag=true, Use Rotational Drag=false)
   &SolveForcesAndVelocity(Clamp Velocity=false, Limit Acceleration=false)
   &ScaleColor(
    Scale Alpha=&Multiply_Float(
     A=&FloatFromCurve(FloatCurve=DataInterface() {propertyValues {Curve {keys [<at -0.0092 0.0144>,<at 0.0967 0.9767>,<at 1.0011 0.002>]}}})
     B=&Waveform(
      "Waveform [1]"=::NewEnumerator0
      "[1] Input Position"=&RandomRangeFloat(Minimum=0, Maximum=1)
      "[1] Frequency"=2
      "[1] Amplitude Min/Max"=<0 1>
     )
    )
    Scale RGB=&VectorFromFloat(Value=&RandomRangeFloat(Minimum=1, Maximum=2.5))
    ScaleA=true
    ScaleRGB=true
   )
  ]
  rendererClasses [&NiagaraSpriteRendererProperties]
  renderers [{rendererClass=&NiagaraSpriteRendererProperties}]
 }
 Steam {
  emitterUpdateScript [
   &EmitterState(Life Cycle Mode=::NewEnumerator1, Loop Duration=1, UseLoopDelay=true)
   &SpawnRate(SpawnRate=4, Use Spawn Probability=false)
  ]
  particleSpawnScript [
   &InitializeParticle(
    Color Maximum=<rgba 0.16 0.7529 0.1506 1>
    Color Minimum=<rgba 0.1506 0.7529 0.1506 1>
    Color Mode=::NewEnumerator2
    Lifetime Min=1.4
    Lifetime Mode=::NewEnumerator1
    Position Offset=<0 0 4>
    Sprite Rotation Mode=::NewEnumerator2
    Sprite Size Mode=::NewEnumerator1
    Uniform Sprite Size Max=50
    Uniform Sprite Size Min=30
    UsePositionOffset=true
    Write Material Random=true
   )
   &AddVelocity(Velocity=&RandomRangeVector(Minimum=<-5 -5 0>, Maximum=<5 5 0>, Evaluation Type=::NewEnumerator1))
   &AddVelocity(Coordinate Space=::World, Velocity=&RandomRangeVector(Minimum=<0 0 5>, Maximum=<0 0 30>)) {module=AddVelocity001}
   /SetVariables(Particles.InitRandomFloat:NiagaraFloat=&RandomRangeFloat(Minimum=0, Maximum=1)) {isSetParameters=true}
   &SystemLocation(MaskSpawnGroup=false) {enabled=false}
   &ShapeLocation(
    Disc Coverage=1
    MaskSpawnGroup=false
    Offset=<0 0 95>
    Offset Mode=::NewEnumerator0
    Override Seed=false
    Ring Radius=30
    Shape Primitive=::NewEnumerator4
    UseOwnerScale=true
   )
  ]
  particleUpdateScript [
   &ParticleState()
   &DynamicMaterialParameters(
    Index 0 Param 1=&FloatFromCurve(
     FloatCurve=DataInterface() {propertyValues {Curve {keys [<at 0 0>,<at 1 1>]}}}
     Scale Curve=&RandomRangeFloat(Minimum=0.1, Maximum=0.3)
    )
    Index 0 Param 4=&FloatFromCurve(FloatCurve=DataInterface() {propertyValues {Curve {keys [<at 0 1>,<at 1 0>]}}}, Scale Curve=0.25)
    Index 2 Param 1=&FloatFromCurve(FloatCurve=DataInterface() {propertyValues {Curve {keys [
         {arriveTangent=2.8586, interpMode=RCIM_Cubic, leaveTangent=2.8586, tangentMode=RCTM_User, time=0, value=0}
         {interpMode=RCIM_Cubic, tangentMode=RCTM_User, time=1, value=1}
        ]}}})
    Index 2 Param 2=&FloatFromCurve(FloatCurve=DataInterface() {propertyValues {Curve {keys [<at 0 1>,<at 1 0>]}}}, Scale Curve=0.2)
    Index 2 Param 3=1
    Index 2 Param 4=0.5
    Module.Param0WriteEnabled=true
    Module.Param0WriteEnabled2=true
    Module.Param1WriteEnabled=false
    Module.Param1WriteEnabled2=true
    Module.Param2WriteEnabled=false
    Module.Param2WriteEnabled2=true
    Module.Param3WriteEnabled=true
    Module.Param3WriteEnabled2=true
    Write Parameter Index 2=true
   )
   &ScaleSpriteSize(Scale Factor=&Vector2DFromFloat(Value=&FloatFromCurve(FloatCurve=DataInterface() {propertyValues {Curve {keys [
          {arriveTangent=1.4006, interpMode=RCIM_Cubic, leaveTangent=1.4006, tangentMode=RCTM_User, time=0, value=0.7}
          {arriveTangent=0.3317, interpMode=RCIM_Cubic, leaveTangent=0.3317, tangentMode=RCTM_User, time=1, value=1.2}
         ]}}})))
   &Drag(Drag=0.5, Use Linear Drag=true, Use Rotational Drag=false)
   &SolveForcesAndVelocity(Clamp Velocity=false, Limit Acceleration=false)
   &ScaleColor(
    Scale Alpha=&FloatFromCurve(
     FloatCurve=DataInterface() {propertyValues {Curve {keys [
         {arriveTangent=4.231, interpMode=RCIM_Cubic, leaveTangent=4.231, tangentMode=RCTM_User, time=0, value=0}
         {arriveTangent=0.1173, interpMode=RCIM_Cubic, leaveTangent=0.1173, tangentMode=RCTM_User, time=0.3, value=1}
         {arriveTangent=-2.9749, interpMode=RCIM_Cubic, leaveTangent=-2.9749, tangentMode=RCTM_User, time=1, value=0}
        ]}}}
     Scale Curve=&Multiply_Float(A=$System.FadeIn, B=0.1)
    )
    ScaleA=true
    ScaleRGB=true
   )
   &self:GoBackToPos(
    PositionGoal:NiagaraPosition=&AddVectorToPosition(Position=$Engine.Owner.Position, Vector=&RandomRangeVector(Minimum=<0 0 10>, Maximum=<1 1 25>))
    ZrestSpeed:NiagaraFloat=&FloatFromCurve(
     FloatCurve=DataInterface() {propertyValues {Curve {keys [<at 0 0.6>,<at 1 0.2>]}}}
     Scale Curve=&MakeCustomFloatFromBool(Bool=$User.bIsFrontEndPreview, True Float=0, False Float=1)
    )
   ) {enabled=false, module=GoBackToPos}
  ]
  rendererClasses [&NiagaraSpriteRendererProperties]
  renderers [
   {rendererClass=&NiagaraSpriteRendererProperties}
   {rendererClass=&NiagaraSpriteRendererProperties, rendererIndex=1}
  ]
 }
 UpperLight {
  emitterUpdateScript [
   &EmitterState(
    Inactive Response=::NewEnumerator1
    Life Cycle Mode=::NewEnumerator1
    Loop Behavior=::NewEnumerator1
    Loop Duration Mode=::NewEnumerator1
    UseLoopDelay=false
   )
   &SpawnBurst_Instantaneous(Use Spawn Probability=false, UseLoopCountLimit=false)
  ]
  particleSpawnScript [
   &InitializeParticle(
    Color Mode=::NewEnumerator0
    Lifetime=0.7
    Mesh Scale Mode=::NewEnumerator3
    Mesh Uniform Scale=2
    UsePositionOffset=false
    Write Material Random=true
   )
   &SystemLocation(MaskSpawnGroup=false, Offset=<0 -10 110>)
  ]
  particleUpdateScript [
   &ParticleState(Kill Particles When Lifetime Has Elapsed=false)
   &FN_Light_Attributes(
    Color=<rgba 0.2356 1 0 1>
    Create Color Attribute=true
    Create Exponent Attribute=true
    Create Radius Attribute=true
    Enable Particle Light Enable=true
    Exponent=5
    Light Enabled=true
    Radius=&Add_Float(A=&Multiply_Float(A=50, B=$System.FadeIn), B=1)
    Scale Color=&Multiply_Float(
     A=&Multiply_Float(A=8, B=$System.FadeIn)
     B=&Waveform(
      Waveform Count=::NewEnumerator1
      "Waveform [1]"=::NewEnumerator0
      "[1] Input Position"=0.5
      "[1] Frequency"=0.7
      "[1] Amplitude Min/Max"=<0.5 1>
      "[1] Scale"=1.5
      "Waveform [2] Blend"=::NewEnumerator2
      "[2] Input Position"=0
      "[2] Phase"=$Emitter.Age
      "[2] Phase Scale"=1
      "[2] Compound Frequency"=2
      "[2] Frequency"=1.2
      "[2] Amplitude Min/Max"=<0.8 1>
      "[2] Scale"=1
      "[2] Bias"=0
      "[2] Exponent"=1
      "[2] Pulse Duration"=0.1
      "[2] Clip Triangle Min/Max"=<0 0>
      "[2] Random Seed"=0
     )
    )
   )
  ]
  rendererClasses [&NiagaraLightRendererProperties, &NiagaraMeshRendererProperties]
  renderers [
   {rendererClass=&NiagaraLightRendererProperties, rendererIndex=0}
   {rendererClass=&NiagaraMeshRendererProperties, rendererIndex=1}
  ]
 }
}
summary {
 systemName=NS_Emote_Cauldron_Idle
 emitters [
  {emitterName=CandyCorn_Fire, rendererClasses [&NiagaraMeshRendererProperties]}
  {emitterName=CauldronOverlay, rendererClasses [&NiagaraMeshRendererProperties]}
  {emitterName=BottomLight, rendererClasses [&NiagaraLightRendererProperties]}
  {emitterName=UpperLight, rendererClasses [&NiagaraLightRendererProperties, &NiagaraMeshRendererProperties]}
  {emitterName=LamaGlow, rendererClasses [&NiagaraSpriteRendererProperties]}
  {emitterName=FallingSmoke, rendererClasses [&NiagaraSpriteRendererProperties]}
  {emitterName=Bubbles, rendererClasses [&NiagaraMeshRendererProperties, &NiagaraSpriteRendererProperties]}
  {emitterName=Steam, rendererClasses [&NiagaraSpriteRendererProperties]}
  {emitterName=Sparkles, rendererClasses [&NiagaraSpriteRendererProperties]}
  {emitterName=BottomGlow, rendererClasses [&NiagaraSpriteRendererProperties]}
 ]
 userVariables [
  User.bIsFrontEnd {type=NiagaraBool}
  User.bIsFrontEndPreview {type=NiagaraBool}
 ]
}
