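# Niagara burst effect for the Cauldron emote: a green magic smoke puff and liquid splash with gold sparkles.
# Smoke, wispy flipbook and curl-noise sparkle sprites rise over splash meshes and falling drops, with smoke adjusted for time of day and front end.
# Goo blobs launch along system-defined splash directions and emit collision events, while the Stickey emitter, lacking its own spawn module, provides lingering splats.
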NiagaraSystem /BRCosmetics/Animation/Game/MainPlayer/Emotes/Cauldron/FX/NS_Emote_Cauldron_MagicSmoke
dependencies {
 usedDataInterfaces [
  &NiagaraDataInterfaceCamera
  &NiagaraDataInterfaceCurve
  &NiagaraDataInterfaceSpriteRendererInfo
  &NiagaraDataInterfaceVectorCurve
  &NiagaraDataInterfaceArrayFloat2
  &NiagaraDataInterfaceVector2DCurve
  &NiagaraDataInterfaceArrayFloat3
 ]
 usedDynamicInputs [&ReturnCameraVector, &ClampFloat]
 usedModules [
  &self:SystemSpawnScript.SystemScriptSource.SystemScriptGraph.NiagaraNodeAssignment_0.SetVariables_#1
  &self:SystemSpawnScript.SystemScriptSource.SystemScriptGraph.NiagaraNodeAssignment_5.SetVariables_#2
  &SystemState
 ]
}
emitters {
 BuffSplash_Mesh {
  emitterUpdateScript [
   &EmitterState()
   &SpawnBurst_Instantaneous(Spawn Count=&RandomRangeInt(Minimum=2, Maximum=2), Use Spawn Probability=false)
  ]
  particleSpawnScript [
   &InitializeParticle(
    Color=&FNTimeOfDayColor(
     Night Color=<rgba 0.136 0.8 0.136 1>
     Morning Color=<rgba 0.136 0.8 0.136 1>
     Day Color=<rgba 0.136 0.8 0.136 1>
     Evening Color=<rgba 0.0333 0.2 0 1>
    )
    Lifetime=0.8
    Mesh Scale Max=<25 25 15>
    Mesh Scale Min=<16 16 8>
    Mesh Scale Mode=::NewEnumerator2
    Position Offset=<0 0 -10>
    UsePositionOffset=true
    Write Material Random=true
   )
   &InitialMeshOrientation(
    Rotation=&RandomRangeVector(Minimum=<-0.05 -0.05 0>, Maximum=<0.05 0.05 1>)
    Use Orientation Vector=true
    Use Rotation Vector=true
   )
   &CameraOffset(Camera Offset Amount=5)
  ]
  particleUpdateScript [
   &ParticleState()
   &InheritSourceMovement(ApplyForce=false, ApplyPosition=true, ApplyPositionFalloff=false)
   &ScaleColor(
    Scale Alpha=&FloatFromCurve(FloatCurve=DataInterface() {propertyValues {Curve {keys [
         {arriveTangent=-0.6531, interpMode=RCIM_Cubic, time=0, value=1}
         {arriveTangent=-0.8638, interpMode=RCIM_Cubic, leaveTangent=-0.8638, tangentMode=RCTM_User, time=0.8, value=0.3}
         {arriveTangent=-0.4827, interpMode=RCIM_Cubic, leaveTangent=-0.4827, tangentMode=RCTM_User, time=1, value=0.1}
        ]}}})
    ScaleA=true
    ScaleRGB=true
   )
   &ScaleMeshSize(Scale Factor=&VectorFromCurve(
     VectorCurve=DataInterface() {propertyValues {
       XCurve {keys [
         {arriveTangent=0.803, interpMode=RCIM_Cubic, leaveTangent=0.803, tangentMode=RCTM_User, time=0, value=0.2}
         {arriveTangent=0.4715, interpMode=RCIM_Cubic, leaveTangent=0.4715, tangentMode=RCTM_User, time=1, value=0.8}
        ]}
       YCurve {keys [
         {arriveTangent=0.803, interpMode=RCIM_Cubic, leaveTangent=0.803, tangentMode=RCTM_User, time=0, value=0.2}
         {arriveTangent=0.4715, interpMode=RCIM_Cubic, leaveTangent=0.4715, tangentMode=RCTM_User, time=1, value=0.8}
        ]}
       ZCurve {keys [
         {arriveTangent=3.0023, interpMode=RCIM_Cubic, leaveTangent=3.0023, tangentMode=RCTM_User, time=0, value=0.3}
         {arriveTangentWeight=0.2848, interpMode=RCIM_Cubic, leaveTangentWeight=0.1221, tangentMode=RCTM_User, time=0.5, value=1}
         {arriveTangent=-5.505, interpMode=RCIM_Cubic, leaveTangent=-5.505, tangentMode=RCTM_User, time=1, value=0}
        ]}
      }}
     Scale Curve=&VectorFromFloat(Value=0.14)
    )) {module=ScaleMeshSize001}
   &SolveForcesAndVelocity(Clamp Velocity=false, Limit Acceleration=false)
   &DynamicMaterialParameters(
    Index 1 Param 2=&Lerp_Float(
     A=0
     B=-220
     Alpha=&FloatFromCurve(FloatCurve=DataInterface() {propertyValues {Curve {keys [
          {interpMode=RCIM_Cubic, tangentMode=RCTM_Break, time=0.111, value=0}
          {arriveTangent=1.6304, interpMode=RCIM_Cubic, tangentMode=RCTM_Break, time=1, value=1}
         ]}}})
    )
    Module.Param0WriteEnabled1=false
    Module.Param1WriteEnabled1=true
    Module.Param2WriteEnabled1=false
    Module.Param3WriteEnabled1=false
    Write Parameter Index 0=false
    Write Parameter Index 1=true
   ) {module=DynamicMaterialParameters002}
  ]
  rendererClasses [&NiagaraMeshRendererProperties]
  renderers [{rendererClass=&NiagaraMeshRendererProperties, rendererIndex=0}]
 }
 Drops003 {
  emitterUpdateScript [
   &EmitterState(
    Life Cycle Mode=::NewEnumerator1
    Loop Behavior=::NewEnumerator1
    Loop Delay=-0.0033
    Loop Duration Mode=::NewEnumerator1
    UseLoopDelay=false
   )
   &SpawnBurst_Instantaneous(
    Spawn Count=&RandomRangeInt(Minimum=14, Maximum=20)
    Spawn Time=0.006
    Use Spawn Probability=false
    UseLoopCountLimit=false
   )
  ]
  particleSpawnScript [
   &InitializeParticle(
    Lifetime=&RandomRangeFloat(Minimum=1, Maximum=2)
    Sprite Size=&Multiply_Vector2D(
     A=&RandomRangeVector2D(Minimum=<4 23>, Maximum=<12 22>)
     B=&Vector2DFromFloat(Value=&RandomRangeFloat(Minimum=0.4, Maximum=1.1))
    )
    Sprite Size Mode=::NewEnumerator4
    UsePositionOffset=false
    Write Material Random=true
   ) {module=InitializeParticle001}
   &FN_SetSpawnGroup(Number of Spawn Groups=6)
   &ShapeLocation(
    Disc Coverage=0.3
    MaskSpawnGroup=false
    Offset Mode=::NewEnumerator0
    Override Seed=false
    Ring Radius=28
    Shape Primitive=::NewEnumerator4
    UseOwnerScale=true
   )
   &AddVelocity(
    Module.Use Falloff Distance=false
    Origin Offset=&RandomRangeVector(Minimum=<0 0 -50>, Maximum=<0 0 -10>)
    Override Seed=false
    Velocity Mode=::NewEnumerator1
    Velocity Speed=&RandomRangeFloat(Minimum=50, Maximum=500)
   )
   /Niagara/Modules/Update/SubUV/SubUVAnimation(SubUV Animation Mode=::NewEnumerator1) {module=SubUVAnimation001}
  ]
  particleUpdateScript [
   &ParticleState()
   &GravityForce(Gravity=<0 0 -1200>)
   &Drag(Use Linear Drag=true, Use Rotational Drag=true)
   &SolveForcesAndVelocity(Clamp Velocity=false, Limit Acceleration=false)
   &ScaleSpriteSize(Scale Factor=&Vector2DFromCurve(
     Vector2Curve=DataInterface() {propertyValues {XCurve {keys [<at -0.0011 1.0067>,<at 0.9997 0.4319>]}, YCurve {keys [<at 0.0102 1.1102>,<at 0.0596 0.8185>,<at 0.9997 0.3098>]}}}
     ScaleCurve=<1 1>
    ))
   &DynamicMaterialParameters(
    Index 0 Param 1=&RandomRangeFloat(Minimum=0, Maximum=1)
    Index 0 Param 2=&FloatFromCurve(FloatCurve=DataInterface() {propertyValues {Curve {keys [<at 0.35 0.45>,<at 1 0.9>]}}})
    Index 0 Param 3=&FloatFromCurve(FloatCurve=DataInterface() {propertyValues {Curve {keys [{arriveTangent=2.6651, interpMode=RCIM_Cubic, leaveTangent=2.6651, tangentMode=RCTM_User, time=0, value=0}, <at 1 1 RCIM_Cubic>]}}})
    Index 0 Param 4=&FloatFromCurve(FloatCurve=DataInterface() {propertyValues {Curve {keys [<at 0 0 RCIM_Cubic>, {arriveTangent=2.332, interpMode=RCIM_Cubic, leaveTangent=2.332, tangentMode=RCTM_User, time=1, value=1}]}}})
    Module.Param0WriteEnabled=true
    Module.Param1WriteEnabled=true
    Module.Param2WriteEnabled=false
    Module.Param3WriteEnabled=true
   ) {module=DynamicMaterialParameters001}
  ]
  rendererClasses [&NiagaraSpriteRendererProperties]
  renderers [{rendererClass=&NiagaraSpriteRendererProperties}]
 }
 Smoke {
  emitterSpawnScript [
   /SetVariables(Emitter.FakeLightActive=true) {isSetParameters=true}
   /SetVariables(Emitter.FrontEndShadowLow:NiagaraFloat=&FN_FrontEndFloatModifier(Gameplay=0.7, Front End=1.5)) {isSetParameters=true}
  ]
  emitterUpdateScript [
   &EmitterState()
   &SpawnBurst_Instantaneous(Spawn Count=6, Use Spawn Probability=false, UseLoopCountLimit=false)
  ]
  particleSpawnScript [
   /SetVariables(Particles.RandomFloat=&RandomRangeFloat(Minimum=0, Maximum=1)) {isSetParameters=true}
   &InitializeParticle(
    Color=<rgba 1 1 1 0.555>
    Lifetime=&RandomRangeFloat(Minimum=0.5, Maximum=1.1)
    Sprite Rotation Angle=&RandomRangeFloat(Minimum=0.3, Maximum=-0.3)
    Sprite Rotation Mode=::NewEnumerator1
    Sprite Size Mode=::NewEnumerator1
    Uniform Sprite Size Max=85
    Uniform Sprite Size Min=75
    UsePositionOffset=true
    Write Material Random=true
   )
   &ShapeLocation(Disc Coverage=0.3, MaskSpawnGroup=false, Override Seed=false, Ring Radius=30, Shape Primitive=::NewEnumerator4, UseOwnerScale=true)
   &AddVelocity(Velocity=<0 0 1>, Velocity Speed Scale=&RandomRangeFloat(Minimum=0, Maximum=250))
   &AddVelocity(
    Module.Use Falloff Distance=false
    Override Seed=false
    Velocity Mode=::NewEnumerator1
    Velocity Speed=&RandomRangeFloat(Minimum=0, Maximum=80)
   ) {module=AddVelocity001}
   &ApplyInitialForces()
  ]
  particleUpdateScript [
   &InheritSourceMovement(ApplyForce=false, ApplyPosition=true, ApplyPositionFalloff=false)
   &DynamicMaterialParameters(
    Index 0 Param 1=&Add_Float(
     A=$Particles.RandomFloat
     B=&FloatFromCurve(
      FloatCurve=DataInterface() {propertyValues {Curve {keys [
          {arriveTangent=0.1552, interpMode=RCIM_Cubic, leaveTangent=0.1552, tangentMode=RCTM_User, time=0, value=1}
          {interpMode=RCIM_Cubic, leaveTangent=0.5473, time=1, value=1.1}
         ]}}}
      Scale Curve=0.5
     )
    )
    Index 0 Param 3=1
    Index 0 Param 4=1
    Index 2 Param 1=&FloatFromCurve(FloatCurve=DataInterface() {propertyValues {Curve {keys [
         {interpMode=RCIM_Cubic, leaveTangent=1.6214, tangentMode=RCTM_Break, time=0.1, value=0}
         {arriveTangent=0.8413, tangentMode=RCTM_User, time=1, value=1}
        ]}}})
    Index 2 Param 2=&FloatFromCurve(
     FloatCurve=DataInterface() {propertyValues {Curve {keys [{arriveTangent=0.7845, interpMode=RCIM_Cubic, leaveTangent=0.7845, tangentMode=RCTM_User, time=0, value=0.2}, <at 1 0.8 RCIM_Cubic>]}}}
     Scale Curve=&RandomRangeFloat(Minimum=0.2, Maximum=0.8)
    )
    Index 2 Param 3=&FloatFromCurve(
     FloatCurve=DataInterface() {propertyValues {Curve {keys [
         {interpMode=RCIM_Cubic, leaveTangent=1.0458, tangentMode=RCTM_Break, time=0, value=1}
         {interpMode=RCIM_Cubic, leaveTangent=11.1233, tangentMode=RCTM_Break, time=1, value=2}
        ]}}}
     Scale Curve=1.25
    )
    Index 2 Param 4=&FloatFromCurve(
     FloatCurve=DataInterface() {propertyValues {Curve {keys [<at 0 1 RCIM_Cubic>, {interpMode=RCIM_Cubic, leaveTangent=-1.2956, tangentMode=RCTM_Break, time=1, value=0.75}]}}}
     Scale Curve=1.5
    )
    Index 3 Param 1=&FloatFromCurve(
     FloatCurve=DataInterface() {propertyValues {Curve {keys [
         <at 0.5 0 RCIM_Cubic>
         {arriveTangent=1, interpMode=RCIM_Cubic, leaveTangent=1, time=1, value=1}
         {arriveTangent=-1, interpMode=RCIM_Cubic, leaveTangent=-1, time=1.5, value=1}
         {interpMode=RCIM_Cubic, leaveTangent=2, time=2, value=0}
        ]}}}
     CurveIndex=$Emitter.Age
     Scale Curve=&Multiply_Float(
      A=&FN_Select_Float_from_Time_Of_Day(Morning=3, Day=3, Evening=3, Night=3)
      B=&FN_FrontEndFloatModifier(Gameplay=1, Front End=2)
     )
    )
    Index 3 Param 2=&Multiply_Float(
     A=&Lerp_Float(
      A=33
      B=75
      Alpha=&FloatFromCurve(
       FloatCurve=DataInterface() {propertyValues {Curve {keys [
           <at 0.5 0 RCIM_Cubic>
           {arriveTangent=1, interpMode=RCIM_Cubic, leaveTangent=1, time=1, value=1}
           {arriveTangent=-1, interpMode=RCIM_Cubic, leaveTangent=-1, time=1.5, value=1}
           <at 2 0 RCIM_Cubic>
          ]}}}
       CurveIndex=$Emitter.Age
      )
     )
     B=&FN_FrontEndFloatModifier(Gameplay=1, Front End=1.25)
    )
    Module.Param0WriteEnabled=true
    Module.Param0WriteEnabled2=true
    Module.Param0WriteEnabled3=true
    Module.Param1WriteEnabled=false
    Module.Param1WriteEnabled2=true
    Module.Param1WriteEnabled3=true
    Module.Param2WriteEnabled=true
    Module.Param2WriteEnabled2=true
    Module.Param2WriteEnabled3=true
    Module.Param3WriteEnabled=true
    Module.Param3WriteEnabled2=true
    Module.Param3WriteEnabled3=true
    Write Parameter Index 2=true
    Write Parameter Index 3=true
   )
   &ParticleState()
   &ScaleSpriteSize(Uniform Curve Sprite Scale=DataInterface() {propertyValues {Curve {keys [
        {arriveTangent=1.6448, interpMode=RCIM_Cubic, leaveTangent=1.6448, tangentMode=RCTM_User, time=0, value=0.4}
        {arriveTangent=0.0224, interpMode=RCIM_Cubic, leaveTangent=0.0224, tangentMode=RCTM_User, time=1, value=1}
       ]}}})
   &ScaleColor(
    Scale Alpha=&FloatFromCurve(
     FloatCurve=DataInterface() {propertyValues {Curve {keys [
         {interpMode=RCIM_Cubic, tangentMode=RCTM_User, time=0, value=0}
         {arriveTangent=0.0617, interpMode=RCIM_Cubic, leaveTangent=0.0617, tangentMode=RCTM_User, time=0.2, value=1}
         {tangentMode=RCTM_User, time=1, value=0}
        ]}}}
     Scale Curve=&Multiply_Float(
      A=&FN_FrontEndFloatModifier(Gameplay=1, Front End=0.666)
      B=&FNTimeOfDayScalar(Night Scale=1, Morning Scale=1, Evening Scale=1)
     )
    )
    Scale RGB=&MakeVectorFromLinearColorRGB(Color=&MultiplyLinearColor(
      Linear Color=&RandomRangeLinearColor(Minimum=<rgba 0.035 0.7 0.0904 1>, Maximum=<rgba 0.1877 0.85 0.1275 1>)
      Scale Factor=&MakeVector4_FromLinearColor(LinearColor=&FNTimeOfDayColor(Night Color=<rgba 1 1 1 1>, Morning Color=<rgba 1 1 1 1>, Day Color=<rgba 1 1 1 1>, Evening Color=<rgba 3 3 3 1>))
     ))
    ScaleA=true
    ScaleRGB=true
   )
   &Drag(Drag=5, Use Linear Drag=true, Use Rotational Drag=true)
   &GravityForce(Gravity=<0 0 -200>)
   &SolveForcesAndVelocity(Clamp Velocity=false, Limit Acceleration=false)
   &CameraOffset(Camera Offset Amount=35)
  ]
  rendererClasses [&NiagaraSpriteRendererProperties]
  renderers [
   {rendererClass=&NiagaraSpriteRendererProperties}
   {rendererClass=&NiagaraSpriteRendererProperties, rendererIndex=1}
  ]
 }
 Sparkles {
  emitterUpdateScript [
   &EmitterState()
   &SpawnBurst_Instantaneous(Spawn Count=&RandomRangeInt(Minimum=10, Maximum=20), Use Spawn Probability=false)
  ]
  particleSpawnScript [
   &InitializeParticle(
    Color=&MultiplyLinearColor(
     Linear Color=<rgba 10 5.875 0.1 1>
     Scale Factor=&MakeFloat4FromVectorAndFloat(
      "Vector (RGB)"=&VectorFromFloat(Value=&RandomRangeFloat(Minimum=1, Maximum=5))
      "Float (Alpha)"=1
     )
    )
    Lifetime=&UniformRangedFloat(Minimum=0.5, Maximum=1.5)
    Sprite Size Mode=::NewEnumerator3
    Uniform Sprite Size=&UniformRangedFloat(Minimum=5, Maximum=10)
    UsePositionOffset=false
    Write Material Random=true
   ) {module=InitializeParticle001}
   &ShapeLocation(
    Disc Coverage=1
    MaskSpawnGroup=false
    Offset Mode=::NewEnumerator0
    Override Seed=false
    Ring Radius=28
    Shape Primitive=::NewEnumerator4
    UseOwnerScale=true
   )
   &AddVelocity(Velocity=&UniformRangedVector(Minimum=<-300 -300 0>, Maximum=<300 300 300>))
   &ApplyInitialForces()
  ]
  particleUpdateScript [
   &ParticleState()
   &InheritSourceMovement(ApplyForce=false, ApplyPosition=true, ApplyPositionFalloff=false)
   &ScaleSpriteSize(Scale Factor=&Vector2DFromFloat(Value=&FloatFromCurve(FloatCurve=DataInterface() {propertyValues {Curve {keys [<at -0.0046 0>,<at 0.1746 0.9952>,<at 1.0011 0.002>]}}}, Scale Curve=0.8)))
   &GravityForce(Gravity=<0 0 200>)
   /Niagara/Modules/Update/Forces/CurlNoiseForce(Noise Quality / Cost=::NewEnumerator0, Noise Strength=500, Pan Noise=false)
   &Drag(Drag=&RandomRangeFloat(Minimum=4, Maximum=6), Use Linear Drag=true, Use Rotational Drag=false)
   &SolveForcesAndVelocity(Clamp Velocity=false, Limit Acceleration=false)
   &ScaleColor(
    Scale Alpha=&Multiply_Float(
     A=&FloatFromCurve(FloatCurve=DataInterface() {propertyValues {Curve {keys [<at -0.0092 0.0144>,<at 0.0967 0.9767>,<at 1.0011 0.002>]}}})
     B=&Waveform(
      "Waveform [1]"=::NewEnumerator0
      "[1] Input Position"=&RandomRangeFloat(Minimum=0, Maximum=1)
      "[1] Frequency"=3
      "[1] Amplitude Min/Max"=<0 1>
     )
    )
    Scale RGB=&VectorFromFloat(Value=&RandomRangeFloat(Minimum=1, Maximum=2.5))
    ScaleA=true
    ScaleRGB=true
   )
   &CameraOffset(Camera Offset Amount=25)
  ]
  rendererClasses [&NiagaraSpriteRendererProperties]
  renderers [{rendererClass=&NiagaraSpriteRendererProperties}]
 }
 Splash_Crown {
  emitterUpdateScript [
   &EmitterState(
    Enable Distance Culling=true
    Life Cycle Mode=::NewEnumerator1
    Loop Behavior=::NewEnumerator1
    Loop Duration Mode=::NewEnumerator1
    MaxDistance=9000
    Module.bUseMaxDistance=true
    Module.bUseMinDistance=false
    Scalability Mode=::NewEnumerator1
    UseLoopDelay=false
   )
   &SpawnBurst_Instantaneous(Use Spawn Probability=false, UseLoopCountLimit=false) {module=SpawnBurst_Instantaneous001}
  ]
  particleSpawnScript [
   &InitializeParticle(
    Color=<rgba 0.05 0.25 0.05 1>
    Lifetime=0.7
    Mesh Scale Mode=::NewEnumerator3
    Mesh Uniform Scale=3.88
    Position Offset=<0 0 -10>
    Sprite Rotation Mode=::NewEnumerator2
    Sprite Size Mode=::NewEnumerator1
    Sprite UV Mode=::NewEnumerator3
    Uniform Sprite Size Max=150
    Uniform Sprite Size Min=120
    UsePositionOffset=true
    Write Material Random=true
   )
   &ShapeLocation(
    Apply To Particle Position=::NewEnumerator1
    MaskSpawnGroup=false
    Offset=<0 0 84>
    Offset Coordinate Space=::World
    Offset Mode=::NewEnumerator0
    Override Seed=false
    Rotation Coordinate Space=::World
    Rotation Mode=::NewEnumerator1
    Sphere Radius=0.1
    UseOwnerScale=true
   ) {module=ShapeLocation001}
   &InitialMeshOrientation(
    Mesh Orientation Mode=::NewEnumerator7
    Rotation=&RandomRangeVector(Minimum=<0 0 0>, Maximum=<0 0 360>)
    Use Rotation Vector=true
   )
   &AddVelocity(Velocity=<0 0 1>, Velocity Speed Scale=50) {module=AddVelocity001}
   &CameraOffset(Camera Offset Amount=15)
  ]
  particleUpdateScript [
   &ParticleState()
   &InheritSourceMovement(ApplyForce=false, ApplyPosition=true, ApplyPositionFalloff=false)
   &ScaleMeshSize(Scale Factor=&Multiply_VectorByFloat(
     Vector=&MakeVector(
      X=1.4
      Y=1.4
      Z=&FloatFromCurve(
       FloatCurve=DataInterface() {propertyValues {Curve {keys [
           {tangentMode=RCTM_User, time=0, value=1.2}
           {arriveTangent=-0.0405, interpMode=RCIM_Cubic, leaveTangent=-0.0405, tangentMode=RCTM_User, time=0.5, value=0.875}
           {arriveTangent=-3.7129, interpMode=RCIM_Cubic, leaveTangent=-3.7129, tangentMode=RCTM_User, time=1, value=0}
          ]}}}
       Scale Curve=1.25
      )
     )
     Float=&FloatFromCurve(
      FloatCurve=DataInterface() {propertyValues {Curve {keys [
          {arriveTangent=3.4838, interpMode=RCIM_Cubic, leaveTangent=3.4838, tangentMode=RCTM_User, time=0, value=0.3}
          {arriveTangent=1.1179, interpMode=RCIM_Cubic, leaveTangent=1.1179, tangentMode=RCTM_User, time=0.2, value=0.7}
          {arriveTangent=0.0224, interpMode=RCIM_Cubic, leaveTangent=0.0224, tangentMode=RCTM_User, time=1, value=1}
         ]}}}
      Scale Curve=1.05
     )
    ))
   &GravityForce(Gravity=<0 0 -100>)
   &SolveForcesAndVelocity(Clamp Velocity=false, Limit Acceleration=false)
   &ScaleColor(
    Scale Alpha=&FloatFromCurve(FloatCurve=DataInterface() {propertyValues {Curve {keys [
         {arriveTangent=-4.4282, interpMode=RCIM_Cubic, leaveTangent=-4.4282, tangentMode=RCTM_User, time=0, value=1}
         {tangentMode=RCTM_User, time=0.75, value=-1}
        ]}}})
    ScaleA=true
    ScaleRGB=true
   )
  ]
  rendererClasses [&NiagaraMeshRendererProperties]
  renderers [{rendererClass=&NiagaraMeshRendererProperties, rendererIndex=0}]
 }
 Splash_Hard001 {
  emitterUpdateScript [
   &EmitterState()
   &SpawnBurst_Instantaneous(Spawn Count=6, Spawn Time=0.006, Use Spawn Probability=false, UseLoopCountLimit=false) {module=SpawnBurst_Instantaneous001}
  ]
  particleSpawnScript [
   &FN_SetSpawnGroup(Number of Spawn Groups=6)
   &InitializeParticle(
    Lifetime Min=2
    Lifetime Mode=::NewEnumerator1
    Position Offset Coordinate Space=::Local
    Sprite Rotation Mode=::NewEnumerator2
    Sprite Size Max=<150 150>
    Sprite Size Min=<80 80>
    Sprite Size Mode=::NewEnumerator2
    Sprite UV Mode=::NewEnumerator3
    UsePositionOffset=true
    Write Material Random=true
   )
   /Niagara/Modules/Update/SubUV/V2/SubUVAnimation(End Frame=3, SubUV Animation Mode:ENiagara_SubUVLookupMode=::NewEnumerator2)
   /SetVariables(Particles.CalculatedDirectionVector:Vector3f=&RotateVectorByQuaternion(
     Vector To Rotate=&Multiply_VectorByFloat(
      Vector=&NormalizeVector(
       Vector To Normalize=&TransformVector(
        Vector=&SelectVectorFromArray(Direct Array Index=$Particles.SpawnGroup, Vector Selection Array=$System.SplashDirections)
        Destination Space=::Simulation
       )
       Safe Normalize=true
       Safe Fallback Threshold=0
       Safe Fallback Vector=<0 0 0.001>
      )
      Float=1
     )
     Quaternion=&MakeQuatFromAxisAngle(
      Axis=<0 0 1>
      AngleInDegrees=&Lerp_Float(A=-15, B=15, Alpha=&RandomRangeFloat(Minimum=0, Maximum=1))
     )
    )) {isSetParameters=true}
   &ShapeLocation(
    MaskSpawnGroup=false
    Offset=<0 0 -10>
    Offset Coordinate Space=::Local
    Offset Mode=::NewEnumerator0
    Override Seed=false
    Shape Origin=&AddVectorToPosition(
     Position=$Particles.Position
     Vector=&Multiply_VectorByFloat(Vector=$Particles.CalculatedDirectionVector, Float=100)
    )
    Sphere Radius=0
    UseOwnerScale=true
   ) {enabled=false}
   /SetVariables(Particles.VisibilityTag=&MakeIntFromFloat(Float=&UniformAOrBFloat(
      A=&UniformAOrBFloat(A=0, Distribution Weight=0, Evaluation Type=::NewEnumerator0, Recalculate Random Each Loop=true, Override Seed=false)
      B=&UniformAOrBFloat(A=2, Distribution Weight=0.3, Evaluation Type=::NewEnumerator0, Recalculate Random Each Loop=true, Override Seed=false)
      Evaluation Type=::NewEnumerator0
      Recalculate Random Each Loop=true
      Override Seed=false
     ))) {isSetParameters=true}
   &AddVelocity(
    Module.Use Falloff Distance=false
    Origin Offset=&MakeVector(X=0, Y=0, Z=&RandomRangeFloat(Minimum=-20, Maximum=-20))
    Override Seed=false
    Velocity Mode=::NewEnumerator1
    Velocity Speed=&Multiply_Float(A=&RandomRangeFloat(Minimum=300, Maximum=400), B=1)
   ) {enabled=false}
   /SetVariables(Particles.Position=&AddVectorToPosition(
     Position=&AddVectorToPosition(Position=$Engine.Emitter.SimulationPosition, Vector=<0 0 0>)
     Vector=&Multiply_VectorByFloat(Vector=$Particles.CalculatedDirectionVector, Float=20)
    )) {isSetParameters=true}
   /SetVariables() {isSetParameters=true}
   /SetVariables(Particles.SpriteSize=&Multiply_Vector2D(
     A=$Particles.SpriteSize
     B=&SelectVector2DFromArray(
      Direct Array Index=$Particles.VisibilityTag
      Vector 2D Selection Array=DataInterface() {propertyValues {FloatData [<1 1>, <0.5 0.5>, <1 1>, <1 1>]}}
     )
    )) {isSetParameters=true}
   &AddVelocity(
    Rotation Coordinate Space=::World
    Velocity=&Add_Vector(
     A=&Multiply_Vector(
      A=$Particles.CalculatedDirectionVector
      B=&VectorFromFloat(Value=&RandomRangeFloat(Minimum=50, Maximum=300))
     )
     B=&RandomRangeVector(Minimum=<0 0 200>, Maximum=<0 0 400>)
    )
   ) {module=AddVelocity001}
   /SetVariables() {isSetParameters=true}
  ]
  particleUpdateScript [
   &ParticleState()
   &InheritSourceMovement(ApplyForce=false, ApplyPosition=true, ApplyPositionFalloff=false)
   &ScaleSpriteSize(Scale Sprite Size Mode=::NewEnumerator0, Uniform Scale Factor=0.6) {module=ScaleSpriteSize001}
   &CameraOffset(Camera Offset Amount=5)
   &GravityForce(Gravity=&Multiply_VectorByFloat(Vector=<0 0 -600>, Float=3)) {module=GravityForce001}
   &Drag(
    Drag=&RandomRangeFloat(Minimum=2, Maximum=1)
    Rotational Drag=50
    Use Linear Drag=true
    Use Rotational Drag=false
   )
   &Collision(
    Advanced Aging Rate=10
    Age Colliding Particles=false
    CPU Trace Vector Length Multiplier=3.5
    Collision Enabled=&Boolean_AndOperation(
     Bool1=&SetBoolByFloatComparison(A=$Particles.Age, B=0.2)
     Bool2=&Boolean_NotOperation(Bool=$Particles.Goo Collided Bool)
    )
    Enable Rest State=false
    Friction=0
    KillOnCollision=false
    Module.EnableMaxCollisionCount=false
    Module.Randomize Collision Normal Vector=false
    Particle Radius Scale=3
    Radius Calculation Type=::NewEnumerator3
    Restitution=-1
   )
   /SetVariables(Particles.Goo Collided Bool:NiagaraBool=$Particles.HasCollided) {isSetParameters=true}
   /Niagara/Modules/Events/GenerateCollisionEvent(
    "Color to Send as Linear Color (ParticleColor)":LinearColor=/Niagara/DynamicInputs/LinearColor/MakeLinearColorFromFloat4(Vector:Vector4f=&MakeVector4(X=$Particles.SubImageIndex, Y=$Particles.SpriteRotation, Z=0, W=0))
    Delay Between Collision Events=0.025
    "Float to Send as Float 2 (RandomNormalizedFloat)":NiagaraFloat=&MakeFloatFromInt(Int=$Particles.VisibilityTag)
    Module.EventDelay=true
    Module.Use Velocity Threshold=true
   )
   &GenerateCollisionEvent(
    Delay Between Collision Events=0.025
    "Float to Send as Float 2 (RandomNormalizedFloat)":NiagaraFloat=&MakeFloatFromInt(Int=$Particles.VisibilityTag)
    Module.EventDelay=true
    Module.Use Velocity Threshold=true
   ) {enabled=false, module=GenerateCollisionEvent001}
   &SolveForcesAndVelocity(Clamp Velocity=false, Limit Acceleration=false)
   &SpriteRotationRate(Rotation Rate=&Multiply_Float(
     A=&FloatFromCurve(
      FloatCurve=DataInterface() {propertyValues {Curve {keys [<at 0 1>,<at 1 0>]}}}
      Scale Curve=&MakeCustomFloatFromBool(Bool=&RandomBool(), True Float=-180, False Float=180)
     )
     B=&RandomRangeFloat(Minimum=1, Maximum=2)
    ))
   &ScaleSpriteSize(Scale Factor=&Vector2DFromCurve(
     Vector2Curve=DataInterface() {propertyValues {
       XCurve {keys [
         {arriveTangent=0.3709, interpMode=RCIM_Cubic, leaveTangent=0.3709, tangentMode=RCTM_User, time=0, value=0.6}
         {arriveTangent=1.4336, tangentMode=RCTM_User, time=0.4, value=1}
        ]}
       YCurve {keys [
         {arriveTangent=1.4008, interpMode=RCIM_Cubic, leaveTangent=1.4008, tangentMode=RCTM_User, time=0, value=0.4}
         {interpMode=RCIM_Cubic, tangentMode=RCTM_User, time=0.75, value=1}
        ]}
      }}
     ScaleCurve=<1 1>
    )) {enabled=false}
   &ScaleSpriteSize(Scale Factor=&Vector2DFromCurve(
     Vector2Curve=DataInterface() {propertyValues {
       XCurve {keys [
         {arriveTangent=1.4008, interpMode=RCIM_Cubic, leaveTangent=1.4008, tangentMode=RCTM_User, time=0, value=0.4}
         {arriveTangentWeight=0.25, interpMode=RCIM_Cubic, tangentMode=RCTM_User, time=0.75, value=1}
        ]}
       YCurve {keys [
         {arriveTangent=1.4008, interpMode=RCIM_Cubic, leaveTangent=1.4008, tangentMode=RCTM_User, time=0, value=0.4}
         {interpMode=RCIM_Cubic, tangentMode=RCTM_User, time=0.75, value=1}
        ]}
      }}
     ScaleCurve=<1 1>
    )) {module=ScaleSpriteSize002}
   &ScaleSpriteSizeBySpeed(Max Scale Factor=<1.2 1.5>, Velocity Threshold=800) {enabled=false}
   &ScaleColor(
    Scale Alpha=&FloatFromCurve(FloatCurve=DataInterface() {propertyValues {Curve {keys [{interpMode=RCIM_Cubic, tangentMode=RCTM_User, time=0.9, value=1}, {interpMode=RCIM_Cubic, tangentMode=RCTM_User, time=1, value=0.7}]}}})
    ScaleA=true
    ScaleRGB=true
   )
   &DynamicMaterialParameters(
    Index 0 Param 2=&FloatFromCurve(
     FloatCurve=DataInterface() {propertyValues {Curve {keys [{arriveTangent=-1.9579, interpMode=RCIM_Cubic, time=0.1, value=0}, {interpMode=RCIM_Cubic, tangentMode=RCTM_User, time=0.3, value=0}]}}}
     Scale Curve=0.5
    )
    Index 0 Param 4=&FloatFromCurve(
     FloatCurve=DataInterface() {propertyValues {Curve {keys [
         {arriveTangent=1.2665, interpMode=RCIM_Cubic, leaveTangent=1.2665, tangentMode=RCTM_User, time=0.1, value=0}
         {arriveTangent=1.0703, interpMode=RCIM_Cubic, leaveTangent=1.0703, tangentMode=RCTM_User, time=0.75, value=2}
        ]}}}
     Scale Curve=&RandomRangeFloat(Minimum=0.5, Maximum=0.8)
    )
    Module.Param0WriteEnabled=false
    Module.Param1WriteEnabled=true
    Module.Param2WriteEnabled=false
    Module.Param3WriteEnabled=true
   )
   /SetVariables(Particles.DispVel:Vector3f=&Lerp_Vector(A=$Particles.Velocity, B=$Particles.Initial.Velocity, Alpha=0.5)) {enabled=false, isSetParameters=true}
   /SetVariables(Particles.DispVel:Vector3f=&Subtract_Position(
     A=$Particles.Initial.Position
     B=&AddVectorToPosition(Position=/Niagara/DynamicInputs/Helpers/SimulationPosition(), Vector=<0 0 -50>)
    )) {enabled=false, isSetParameters=true}
   /SetVariables(Particles.Dynam 2:Vector4f=&MakeVector4(
     X=&RandomRangeFloat(Minimum=0, Maximum=1)
     Y=&FloatFromCurve(
      FloatCurve=DataInterface() {propertyValues {Curve {keys [{arriveTangent=-1.9579, interpMode=RCIM_Cubic, time=0.1, value=0.2}, {interpMode=RCIM_Cubic, tangentMode=RCTM_User, time=0.3, value=0}]}}}
      Scale Curve=0.5
     )
     Z=1
     W=&FloatFromCurve(FloatCurve=DataInterface() {propertyValues {Curve {keys [
          {
           arriveTangentWeight=0.4105
           interpMode=RCIM_Cubic
           leaveTangentWeight=0.3333
           tangentMode=RCTM_User
           tangentWeightMode=RCTWM_WeightedBoth
           time=0
           value=0.5593
          }
          {
           arriveTangentWeight=0.3333
           interpMode=RCIM_Cubic
           leaveTangentWeight=0.7342
           tangentMode=RCTM_User
           tangentWeightMode=RCTWM_WeightedBoth
           time=1
           value=0.8836
          }
         ]}}})
    )) {isSetParameters=true}
  ]
  rendererClasses [&NiagaraSpriteRendererProperties]
  renderers [
   {rendererClass=&NiagaraSpriteRendererProperties}
   {rendererClass=&NiagaraSpriteRendererProperties, rendererIndex=1}
   {rendererClass=&NiagaraSpriteRendererProperties, rendererIndex=2}
   {rendererClass=&NiagaraSpriteRendererProperties, rendererIndex=3}
  ]
 }
 Stickey {
  emitterSpawnScript [/SetVariables(Emitter.Mat Rand:NiagaraFloat=&RandomRangeFloat(Minimum=0, Maximum=1)) {isSetParameters=true}]
  emitterUpdateScript [&EmitterState()]
  particleSpawnScript [
   &InitializeParticle(
    Lifetime Min=2.5
    Lifetime Mode=::NewEnumerator1
    Position Offset=<0 0 1>
    Position Offset Coordinate Space=::Local
    Sprite Size=&Multiply_Vector2D_ByFloat(Float=0.555, Vector2D=&RandomRangeVector2D(Minimum=<80 80>, Maximum=<150 150>))
    Sprite Size Mode=::NewEnumerator4
    Sprite UV Mode=::NewEnumerator3
    UsePositionOffset=true
    Write Material Random=true
   )
   &SubUVAnimation(End Frame=3, SubUV Animation Mode:ENiagara_SubUVLookupMode=::NewEnumerator2) {enabled=false, module=SubUVAnimation001}
   /SetVariables() {isSetParameters=true}
   &InitialMeshOrientation(Use Rotation Vector=false)
   /SetVariables(Particles.DecalFade=0.5) {isSetParameters=true}
  ]
  particleUpdateScript [
   &ParticleState()
   &ScaleColor(
    Scale Alpha=&FloatFromCurve(
     FloatCurve=DataInterface() {propertyValues {Curve {keys [
         {arriveTangentWeight=0.2357, interpMode=RCIM_Cubic, leaveTangentWeight=0.2357, tangentMode=RCTM_User, time=0.9, value=1}
         {tangentMode=RCTM_User, time=1, value=0}
        ]}}}
     Scale Curve=25
    )
    ScaleA=true
    ScaleRGB=true
   )
   &DynamicMaterialParameters(
    Index 0 Param 2=&FloatFromCurve(
     FloatCurve=DataInterface() {propertyValues {Curve {keys [{arriveTangent=-1.9579, interpMode=RCIM_Cubic, time=0.2, value=0}, {interpMode=RCIM_Cubic, tangentMode=RCTM_User, time=0.45, value=0}]}}}
     Scale Curve=0.5
    )
    Index 0 Param 4=&FloatFromCurve(
     FloatCurve=DataInterface() {propertyValues {Curve {keys [
         {arriveTangent=-25.3206, interpMode=RCIM_Cubic, leaveTangent=-25.3206, tangentMode=RCTM_User, time=0, value=2}
         {arriveTangentWeight=0.0667, interpMode=RCIM_Cubic, leaveTangentWeight=0.2667, tangentMode=RCTM_User, time=0.05, value=0.5}
         {arriveTangentWeight=0.359, interpMode=RCIM_Cubic, leaveTangentWeight=0.359, tangentMode=RCTM_User, time=0.5, value=0.5}
         {arriveTangent=4.9537, interpMode=RCIM_Cubic, leaveTangent=4.9537, tangentMode=RCTM_User, time=1, value=2}
        ]}}}
     Scale Curve=&RandomRangeFloat(Minimum=0.5, Maximum=0.8)
    )
    Module.Param0WriteEnabled=false
    Module.Param1WriteEnabled=true
    Module.Param2WriteEnabled=false
    Module.Param3WriteEnabled=true
   )
   /SetVariables(Particles.Dynam 2:Vector4f=&MakeVector4(
     X=&RandomRangeFloat(Minimum=0, Maximum=1)
     Y=&FloatFromCurve(
      FloatCurve=DataInterface() {propertyValues {Curve {keys [{arriveTangent=-1.9579, interpMode=RCIM_Cubic, time=0.1, value=0.2}, {interpMode=RCIM_Cubic, tangentMode=RCTM_User, time=0.3, value=0}]}}}
      Scale Curve=0.5
     )
     Z=0
     W=&FloatFromCurve(FloatCurve=DataInterface() {propertyValues {Curve {keys [
          {
           arriveTangent=-2.2031
           arriveTangentWeight=0.0864
           interpMode=RCIM_Cubic
           leaveTangent=-2.2031
           leaveTangentWeight=0.0864
           tangentMode=RCTM_User
           tangentWeightMode=RCTWM_WeightedBoth
           time=0
           value=1
          }
          {arriveTangentWeight=0.05, interpMode=RCIM_Cubic, leaveTangentWeight=0.2833, tangentMode=RCTM_User, time=0.2, value=0.7}
          {arriveTangentWeight=0.1167, interpMode=RCIM_Cubic, leaveTangentWeight=0.1667, tangentMode=RCTM_User, time=0.5, value=0.7}
          {arriveTangent=4.9537, arriveTangentWeight=0.1667, interpMode=RCIM_Cubic, leaveTangent=4.9537, tangentMode=RCTM_User, time=1, value=2}
         ]}}})
    )) {isSetParameters=true}
  ]
  rendererClasses [&NiagaraSpriteRendererProperties]
  renderers [
   {rendererClass=&NiagaraSpriteRendererProperties}
   {rendererClass=&NiagaraSpriteRendererProperties, rendererIndex=1}
   {rendererClass=&NiagaraSpriteRendererProperties, rendererIndex=2}
   {rendererClass=&NiagaraSpriteRendererProperties, rendererIndex=3}
  ]
 }
 WispySmoke {
  emitterUpdateScript [
   &EmitterState()
   &SpawnBurst_Instantaneous(Spawn Count=4, Use Spawn Probability=false, UseLoopCountLimit=false)
  ]
  particleSpawnScript [
   &InitializeParticle(
    Color=<rgba 1 1 1 0.5>
    Lifetime=&RandomRangeFloat(Minimum=0.5, Maximum=0.8)
    Sprite Rotation Mode=::NewEnumerator2
    Sprite Size Mode=::NewEnumerator3
    Uniform Sprite Size=100
    UsePositionOffset=false
    Write Material Random=true
   )
   &ShapeLocation(
    Disc Coverage=0.3
    MaskSpawnGroup=false
    Offset Mode=::NewEnumerator0
    Override Seed=false
    Ring Radius=30
    Shape Primitive=::NewEnumerator4
    UseOwnerScale=true
   )
   &AddVelocity(Velocity=<0 0 1>, Velocity Speed Scale=&RandomRangeFloat(Minimum=0, Maximum=150))
   /SetVariables(Particles.RandomFrameOffset:NiagaraInt32=&RandomRangeInt(Minimum=15, Maximum=22)) {isSetParameters=true}
  ]
  particleUpdateScript [
   &Drag(Drag=5, Use Linear Drag=true, Use Rotational Drag=true) {module=Drag001}
   &InheritSourceMovement(ApplyForce=false, ApplyPosition=true, ApplyPositionFalloff=false)
   &GravityForce(Gravity=<0 0 -150>)
   &SolveForcesAndVelocity(Clamp Velocity=false, Limit Acceleration=false)
   &ParticleState()
   &ScaleSpriteSize(Uniform Curve Sprite Scale=DataInterface() {propertyValues {Curve {keys [<at 0 0.5 RCIM_Cubic>, {arriveTangent=0.5164, interpMode=RCIM_Cubic, leaveTangent=0.5164, tangentMode=RCTM_User, time=1, value=1}]}}})
   &SubUVAnimation(
    End Frame Range Override:NiagaraInt32=&Add_Integer(A=$Particles.RandomFrameOffset, B=10)
    Start Frame Range Override=$Particles.RandomFrameOffset
    UseEndFrame=true
    UseStartFrame=true
   )
   &SolveForcesAndVelocity(Clamp Velocity=false, Limit Acceleration=false) {module=SolveForcesAndVelocity001}
   &ScaleColor(
    Scale Alpha=&FloatFromCurve(FloatCurve=DataInterface() {propertyValues {Curve {keys [
         {interpMode=RCIM_Cubic, tangentMode=RCTM_User, time=0, value=0}
         {interpMode=RCIM_Cubic, tangentMode=RCTM_User, time=0.1, value=1}
         {tangentMode=RCTM_User, time=1, value=0}
        ]}}})
    Scale RGB=&MakeVectorFromLinearColorRGB(Color=&MultiplyLinearColor(
      Linear Color=&RandomRangeLinearColor(Minimum=<rgba 0.1275 0.85 0.1877 1>, Maximum=<rgba 0.1877 0.85 0.1275 1>)
      Scale Factor=&MakeVector4(X=1.5, Y=1.5, Z=1.5, W=&RandomRangeFloat(Minimum=1, Maximum=2))
     ))
    ScaleA=true
    ScaleRGB=true
   )
   &CameraOffset(Camera Offset Amount=25)
  ]
  rendererClasses [&NiagaraSpriteRendererProperties]
  renderers [{rendererClass=&NiagaraSpriteRendererProperties}]
 }
}
summary {
 systemName=NS_Emote_Cauldron_MagicSmoke
 emitters [
  {emitterName=Smoke, rendererClasses [&NiagaraSpriteRendererProperties]}
  {emitterName=WispySmoke, rendererClasses [&NiagaraSpriteRendererProperties]}
  {emitterName=Sparkles, rendererClasses [&NiagaraSpriteRendererProperties]}
  {emitterName=BuffSplash_Mesh, rendererClasses [&NiagaraMeshRendererProperties]}
  {emitterName=Stickey, rendererClasses [&NiagaraSpriteRendererProperties]}
  {emitterName=Splash_Crown, rendererClasses [&NiagaraMeshRendererProperties]}
  {emitterName=Splash_Hard001, rendererClasses [&NiagaraSpriteRendererProperties]}
  {emitterName=Drops003, rendererClasses [&NiagaraSpriteRendererProperties]}
 ]
 userVariables [
  User.bIsFrontEnd {type=NiagaraBool}
  User.bIsFrontEndPreview {type=NiagaraBool}
  User.FadeAllFlame {type=NiagaraFloat, defaultValue=0, description=""}
  User.PlayerSpeed {type=NiagaraFloat, defaultValue=0, description=""}
 ]
}
